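# Chaos Cloth asset for the M_MED_Gimmick medium male player body, built by an embedded Dataflow graph.
# It merges an imported sim collection with one render section from a skeletal mesh variable, then proxy-deforms all render vertices.
# Legacy-named weight masks drive max distance, backstop, anim drive, drag and lift within a chain of PBD configs.
# Both self-collision branches are off; the chain adds body collision and a pelvis reference before the LOD0 terminal.
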ChaosClothAsset /BRCosmetics/Characters/Player/Male/Medium/Bodies/M_MED_Gimmick/Meshes/ClothAsset/ClothApplesToApples/M_MED_Gimmick_ClothApples2Apples
name=EmbeddedDataflow
graph=&self:EmbeddedDataflow
connections [
 {
  fromNode=SkinningBlend
  fromPinGuid=#1
  toNode=SimulationLongRangeAttachmentConfig
  toPinGuid=#2
 }
 {
  fromNode=SimulationMaxDistanceConfig
  fromPin=KinematicVertices3D
  fromPinGuid=#3
  toNode=SkinningBlend
  toPin=KinematicVertices3D.StringValue
  toPinGuid=#4
 }
 {
  fromNode=SimulationMaxDistanceConfig
  fromPin=KinematicVertices3D
  fromPinGuid=#3
  toNode=SimulationLongRangeAttachmentConfig
  toPin=FixedEndSet.StringValue
  toPinGuid=#5
 }
 {
  fromNode=LegacyMask_BackstopRadius
  fromPin=OutputName.StringValue
  fromPinGuid=#6
  toNode=SimulationBackstopConfig
  toPin=BackstopRadius.WeightMap
  toPinGuid=#7
 }
 {
  fromNode=SimulationLongRangeAttachmentConfig
  fromPinGuid=#8
  toNode=SimulationMassConfig
  toPinGuid=#9
 }
 {
  fromNode=SimulationStretchConfig
  fromPinGuid=#10
  toNode=SimulationBendingConfig
  toPinGuid=#11
 }
 {
  fromNode=ReRouteNode
  fromPin=Value
  fromPinGuid=#12
  toNode=UseSelfCollisionSpheres
  toPin=FalseCollection
  toPinGuid=#13
 }
 {
  fromNode=SimulationVelocityScaleConfig
  fromPinGuid=#14
  toNode=SetPhysicsAsset
  toPinGuid=#15
 }
 {
  fromNode=SimulationMassConfig
  fromPinGuid=#16
  toNode=SimulationStretchConfig
  toPinGuid=#17
 }
 {
  fromNode=SimulationAnimDriveConfig
  fromPinGuid=#18
  toNode=SimulationVelocityScaleConfig
  toPinGuid=#19
 }
 {
  fromNode=ReRouteNode_0
  fromPin=Value
  fromPinGuid=#20
  toNode=SimulationSelfCollisionConfig
  toPinGuid=#21
 }
 {
  fromNode=SimulationSolverConfig LOD0
  fromPinGuid=#22
  toNode=StripUserAttributes
  toPinGuid=#23
 }
 {fromNode=SetPhysicsAsset, fromPinGuid=#24, toNode=ReferenceBone, toPinGuid=#25}
 {
  fromNode=ReferenceBone
  fromPinGuid=#26
  toNode=SimulationSolverConfig LOD0
  toPinGuid=#27
 }
 {
  fromNode=SimulationBendingConfig
  fromPinGuid=#28
  toNode=SimulationCollisionConfig
  toPinGuid=#29
 }
 {
  fromNode=SimulationSelfCollisionConfig
  fromPinGuid=#30
  toNode=UseSelfCollisions
  toPin=TrueCollection
  toPinGuid=#31
 }
 {
  fromNode=UseSelfCollisionSpheres
  fromPin=ChosenCollection
  fromPinGuid=#32
  toNode=UseSelfCollisions
  toPin=FalseCollection
  toPinGuid=#33
 }
 {
  fromNode=SimulationSelfCollisionSpheresConfig
  fromPinGuid=#34
  toNode=UseSelfCollisionSpheres
  toPin=TrueCollection
  toPinGuid=#35
 }
 {
  fromNode=ReRouteNode_0
  fromPin=Value
  fromPinGuid=#20
  toNode=SimulationSelfCollisionSpheresConfig
  toPinGuid=#36
 }
 {
  fromNode=ReRouteNode_0
  fromPin=Value
  fromPinGuid=#20
  toNode=ReRouteNode
  toPin=Value
  toPinGuid=#37
 }
 {
  fromNode=LegacyMask_MaxDistance
  fromPin=OutputName.StringValue
  fromPinGuid=#38
  toNode=SimulationMaxDistanceConfig
  toPin=MaxDistance.WeightMap
  toPinGuid=#39
 }
 {
  fromNode=SimulationAerodynamicsConfig
  fromPinGuid=#40
  toNode=SimulationGravityConfig
  toPinGuid=#41
 }
 {
  fromNode=ProceduralSelection_0
  fromPinGuid=#42
  toNode=ProceduralSelection
  toPinGuid=#43
 }
 {
  fromNode=ImportedSimClothCollection
  fromPin=Value
  fromPinGuid=#44
  toNode=MergeClothCollections
  toPin="Collections[0]"
  toPinGuid=#45
 }
 {
  fromNode=SkeletalMeshImport_Render
  fromPinGuid=#46
  toNode=MergeClothCollections
  toPin="Collections[1]"
  toPinGuid=#47
 }
 {
  fromNode=SimulationCollisionConfig
  fromPinGuid=#48
  toNode=ReRouteNode_0
  toPin=Value
  toPinGuid=#49
 }
 {
  fromNode=RenderSkeletalMesh
  fromPin=Value
  fromPinGuid=#50
  toNode=SkeletalMeshImport_Render
  toPin=SkeletalMesh
  toPinGuid=#51
 }
 {
  fromNode=SimulationGravityConfig
  fromPinGuid=#52
  toNode=SimulationPressureConfig
  toPinGuid=#53
 }
 {
  fromNode=SimulationBackstopConfig
  fromPinGuid=#54
  toNode=SimulationDampingConfig
  toPinGuid=#55
 }
 {
  fromNode=StripUserAttributes
  fromPinGuid=#56
  toNode=ClothAssetTerminal
  toPin="CollectionLods[0]"
  toPinGuid=#57
 }
 {
  fromNode=LegacyMask_AnimDriveStiffness
  fromPin=OutputName.StringValue
  fromPinGuid=#58
  toNode=SimulationAnimDriveConfig
  toPin=AnimDriveStiffness.WeightMap
  toPinGuid=#59
 }
 {
  fromNode=MergeClothCollections
  fromPinGuid=#60
  toNode=ProceduralSelection_0
  toPinGuid=#61
 }
 {fromNode=ProceduralSelection, fromPinGuid=#62, toNode=ProxyDeformer, toPinGuid=#63}
 {
  fromNode=ProceduralSelection_0
  fromPin=OutputName
  fromPinGuid=#64
  toNode=ProxyDeformer
  toPin="SelectionFilterSets[0].RenderSelection.StringValue"
  toPinGuid=#65
 }
 {
  fromNode=ProceduralSelection
  fromPin=OutputName
  fromPinGuid=#66
  toNode=ProxyDeformer
  toPin="SelectionFilterSets[0].SimSelection.StringValue"
  toPinGuid=#67
 }
 {
  fromNode=SimulationMaxDistanceConfig
  fromPinGuid=#68
  toNode=SkinningBlend
  toPinGuid=#69
 }
 {
  fromNode=ProxyDeformer
  fromPinGuid=#70
  toNode=LegacyMask_MaxDistance
  toPinGuid=#71
 }
 {
  fromNode=LegacyMask_MaxDistance
  fromPinGuid=#72
  toNode=SimulationMaxDistanceConfig
  toPinGuid=#73
 }
 {
  fromNode=UseSelfCollisions
  fromPin=ChosenCollection
  fromPinGuid=#74
  toNode=LegacyMask_BackstopDistance
  toPinGuid=#75
 }
 {
  fromNode=LegacyMask_BackstopDistance
  fromPin=OutputName.StringValue
  fromPinGuid=#76
  toNode=SimulationBackstopConfig
  toPin=BackstopDistance.WeightMap
  toPinGuid=#77
 }
 {
  fromNode=LegacyMask_BackstopDistance
  fromPinGuid=#78
  toNode=LegacyMask_BackstopRadius
  toPinGuid=#79
 }
 {
  fromNode=LegacyMask_BackstopRadius
  fromPinGuid=#80
  toNode=SimulationBackstopConfig
  toPinGuid=#81
 }
 {
  fromNode=SimulationPressureConfig
  fromPinGuid=#82
  toNode=LegacyMask_AnimDriveStiffness
  toPinGuid=#83
 }
 {
  fromNode=LegacyMask_AnimDriveStiffness
  fromPinGuid=#84
  toNode=SimulationAnimDriveConfig
  toPinGuid=#85
 }
 {
  fromNode=SimulationDampingConfig
  fromPinGuid=#86
  toNode=LegacyMask_Drag
  toPinGuid=#87
 }
 {
  fromNode=LegacyMask_Drag
  fromPin=OutputName.StringValue
  fromPinGuid=#88
  toNode=SimulationAerodynamicsConfig
  toPin=Drag.WeightMap
  toPinGuid=#89
 }
 {fromNode=LegacyMask_Drag, fromPinGuid=#90, toNode=LegacyMask_Lift, toPinGuid=#91}
 {
  fromNode=LegacyMask_Lift
  fromPinGuid=#92
  toNode=SimulationAerodynamicsConfig
  toPinGuid=#93
 }
 {
  fromNode=LegacyMask_Lift
  fromPin=OutputName.StringValue
  fromPinGuid=#94
  toNode=SimulationAerodynamicsConfig
  toPin=Lift.WeightMap
  toPinGuid=#95
 }
]
nodes [
 SimulationMaxDistanceConfig {
  type=FChaosClothAssetSimulationMaxDistanceConfigNode
  guid=#96
  properties {MaxDistance=(bIsAnimatable=True,High=200.224,WeightMap="MaxDistance")}
 }
 SimulationLongRangeAttachmentConfig {
  type=FChaosClothAssetSimulationLongRangeAttachmentConfigNode_v2
  guid=#97
  inputPins [
   Collection {type=FManagedArrayCollection}
   FixedEndSet.StringValue {type=FString}
   TetherStiffness.WeightMap {type=FString}
   TetherScale.WeightMap {type=FString}
   "CustomTetherData[0].CustomFixedEndSet.StringValue" {type=FString}
   "CustomTetherData[0].CustomDynamicEndSet.StringValue" {type=FString}
  ]
  outputPins [Collection {type=FManagedArrayCollection}]
  properties {
   CustomTetherData=(())
   FixedEndSet=(StringValue="KinematicVertices3D")
   TetherScale=(bIsAnimatable=True,Low=1.05,High=1.05,WeightMap="TetherScale")
   TetherStiffness=(bIsAnimatable=True,Low=1,High=1,WeightMap="TetherStiffness")
   bEnableCustomTetherGeneration=False
   bUseGeodesicTethers=True
   bWarnDuplicateProperty=True
  }
 }
 SetPhysicsAsset {
  type=FChaosClothAssetSetPhysicsAssetNode
  guid=#98
  properties {PhysicsAsset=/Script/Engine.PhysicsAsset'/BRCosmetics/Characters/Player/Male/Medium/Bodies/M_MED_Gimmick/Meshes/M_MED_Gimmick_ClothCollision'}
 }
 SimulationStretchConfig {
  type=FChaosClothAssetSimulationStretchConfigNode
  guid=#99
  properties {
   SolverType=PBD
   StretchAnisoDamping=(bIsAnimatable=True,Low=0.0001,High=0.0001,WeightMap="StretchAnisoDamping")
   StretchStiffnessBias=(bIsAnimatable=True,Low=12.9211,High=502.4609,WeightMap="StretchStiffnessBias")
   StretchStiffnessWarp=(bIsAnimatable=True,Low=159.1384,High=490.6552,WeightMap="StretchStiffnessWarp")
   StretchStiffnessWeft=(bIsAnimatable=True,Low=254.5566,High=629.4298,WeightMap="StretchStiffnessWeft")
  }
 }
 SimulationBendingConfig {
  type=FChaosClothAssetSimulationBendingConfigNode
  guid=#100
  inputPins [
   Collection {type=FManagedArrayCollection}
   FlatnessRatio.WeightMap {type=FString}
   RestAngle.WeightMap {type=FString}
   BendingStiffness.WeightMap {type=FString}
   BendingStiffnessWarp.WeightMap {type=FString}
   BendingStiffnessWeft.WeightMap {type=FString}
   BendingStiffnessBias.WeightMap {type=FString}
   BendingDamping.WeightMap {type=FString}
   BendingAnisoDamping.WeightMap {type=FString}
   AnisoBucklingRatio.WeightMap {type=FString}
   BucklingStiffness.WeightMap {type=FString}
   BucklingStiffnessWarp.WeightMap {type=FString}
   BucklingStiffnessWeft.WeightMap {type=FString}
   BucklingStiffnessBias.WeightMap {type=FString}
   BucklingRatioWeighted.WeightMap {type=FString}
  ]
  properties {
   AnisoBucklingRatio=(bIsAnimatable=True,Low=0.3,High=0.3,WeightMap="BucklingRatio")
   BendingAnisoDamping=(bIsAnimatable=True,Low=0.0001,High=0.0001,WeightMap="BendingAnisoDamping")
   BendingStiffnessBias=(bIsAnimatable=True,Low=0.025,High=0.3068,WeightMap="BendingStiffnessBias")
   BendingStiffnessWarp=(bIsAnimatable=True,Low=0.0305,High=13,WeightMap="BendingStiffnessWarp")
   BendingStiffnessWeft=(bIsAnimatable=True,Low=0.0196,High=0.1843,WeightMap="BendingStiffnessWeft")
   BucklingRatioWeighted=(bIsAnimatable=True,WeightMap="BucklingRatio")
   BucklingStiffness=(bIsAnimatable=True,Low=1,High=1,WeightMap="BucklingStiffness")
   BucklingStiffnessBias=(bIsAnimatable=True,Low=0.0062,High=0.0767,WeightMap="BucklingStiffnessBias")
   BucklingStiffnessWarp=(bIsAnimatable=True,Low=0.0076,High=3.25,WeightMap="BucklingStiffnessWarp")
   BucklingStiffnessWeft=(bIsAnimatable=True,Low=0.0049,High=0.0461,WeightMap="BucklingStiffnessWeft")
   ConstraintType=FacesSpring
   FlatnessRatio=(High=1,WeightMap="FlatnessRatio")
   SolverType=PBD
  }
 }
 SimulationCollisionConfig {
  type=FChaosClothAssetSimulationCollisionConfigNode
  guid=#101
  properties {
   ClothCollisionThicknessImported=(bIsAnimatable=True,WeightMap="ClothCollisionThickness")
   CollisionThicknessImported=()
   FrictionCoefficientWeighted=(bIsAnimatable=True,Low=0.2,High=0.2,WeightMap="FrictionCoefficient")
   InnerCollisionThickness=(WeightMap="InnerCollisionThickness")
   bUseCCD=False
  }
 }
 SimulationMassConfig {
  type=FChaosClothAssetSimulationMassConfigNode
  guid=#102
  properties {DensityWeighted=(Low=1,High=1,WeightMap="Density")}
 }
 SimulationAerodynamicsConfig {
  type=FChaosClothAssetSimulationAerodynamicsConfigNode
  guid=#103
  properties {
   Drag=(bIsAnimatable=True,Low=2,High=2,WeightMap="Drag")
   Lift=(bIsAnimatable=True,Low=0.5,High=0.5,WeightMap="Lift")
   TurbulenceRatio=1
   WindVelocity=(X=0,Y=0,Z=0)
   WindVelocitySpace=WorldSpace
  }
 }
 SimulationDampingConfig {
  type=FChaosClothAssetSimulationDampingConfigNode
  guid=#104
  properties {
   DampingCoefficientWeighted=(bIsAnimatable=True,WeightMap="DampingCoefficient")
   LocalDampingAngularCoefficient=0.1
   LocalDampingLinearCoefficient=0.1
   LocalDampingSpace=CenterOfMass
  }
 }
 SimulationSelfCollisionSpheresConfig {
  type=FChaosClothAssetSimulationSelfCollisionSpheresConfigNode
  guid=#105
  inputPins [Collection {type=FManagedArrayCollection}]
  outputPins [
   Collection {type=FManagedArrayCollection}
   SelfCollisionSphereSetName {type=FString}
  ]
  properties {
   SelfCollisionSphereRadius=0.5
   SelfCollisionSphereRadiusCullMultiplier=1
   SelfCollisionSphereSetName=SelfCollisionSphereSetName
   SelfCollisionSphereStiffness=1
   bWarnDuplicateProperty=True
  }
 }
 ClothAssetTerminal {type=FChaosClothAssetTerminalNode_v2, guid=#106}
 SimulationSolverConfig LOD0 {
  type=FChaosClothAssetSimulationSolverConfigNode
  guid=#107
  properties {MaxNumIterations=10, NumIterations=1, NumSubstepsImported=(ImportedValue=1)}
 }
 StripUserAttributes {
  type=FChaosClothStripUserAttributesNode
  guid=#108
  inputPins [
   Collection {type=FManagedArrayCollection}
   ExtraSimWeightMaps {type="TArray<FString>"}
   ExtraSets {type="TArray<FString>"}
   ExtraSimFaceIntMaps {type="TArray<FString>"}
   ExtraRenderWeightMaps {type="TArray<FString>"}
  ]
  outputPins [Collection {type=FManagedArrayCollection}]
  properties {
   ExtraRenderWeightMaps=""
   ExtraSets=""
   ExtraSimFaceIntMaps=""
   ExtraSimWeightMaps=""
   bCopyAllUserAttributesToSimModel=True
   bStripUnusedRenderWeightMaps=True
   bStripUnusedSets=True
   bStripUnusedSimFaceIntMaps=True
   bStripUnusedSimWeightMaps=True
  }
 }
 SimulationSelfCollisionConfig {
  type=FChaosClothAssetSimulationSelfCollisionConfigNode_v2
  guid=#109
  inputPins [
   Collection {type=FManagedArrayCollection}
   SelfCollisionLayers.StringValue {type=FString}
   SelfCollisionDisabledFaces.StringValue {type=FString}
   SelfCollisionThickness.WeightMap {type=FString}
  ]
  outputPins [Collection {type=FManagedArrayCollection}]
  properties {
   NumContourMinimizationPostSteps=0
   SelfCollisionDisableNeighborDistance=5
   SelfCollisionDisabledFaces=(StringValue="SelfCollisionDisabledFaces")
   SelfCollisionFriction=()
   SelfCollisionLayers=(StringValue="SelfCollisionLayers")
   SelfCollisionStiffness=0.5
   SelfCollisionThickness=(bIsAnimatable=True,Low=2,High=2,WeightMap="SelfCollisionThickness")
   bUseContourMinimization=True
   bUseGlobalIntersectionAnalysis=True
   bUseGlobalPostStepContours=True
   bUseSelfCollisions=True
   bUseSelfIntersections=False
   bWarnDuplicateProperty=True
  }
 }
 UseSelfCollisions {
  type=FBranchCollectionDataflowNode
  guid=#110
  inputPins [
   TrueCollection {type=FManagedArrayCollection}
   FalseCollection {type=FManagedArrayCollection}
   bCondition {type=bool}
  ]
  outputPins [ChosenCollection {type=FManagedArrayCollection}]
  properties {bCondition=False}
 }
 UseSelfCollisionSpheres {
  type=FBranchCollectionDataflowNode
  guid=#111
  inputPins [
   TrueCollection {type=FManagedArrayCollection}
   FalseCollection {type=FManagedArrayCollection}
   bCondition {type=bool}
  ]
  outputPins [ChosenCollection {type=FManagedArrayCollection}]
  properties {bCondition=False}
 }
 ReRouteNode {
  type=FDataflowReRouteNode
  guid=#112
  inputPins [Value {type=FManagedArrayCollection}]
  outputPins [Value {type=FManagedArrayCollection}]
 }
 SimulationBackstopConfig {
  type=FChaosClothAssetSimulationBackstopConfigNode
  guid=#113
  inputPins [
   Collection {type=FManagedArrayCollection}
   BackstopDistance.WeightMap {type=FString}
   BackstopRadius.WeightMap {type=FString}
   BackstopMeshName.StringValue {type=FString}
  ]
  outputPins [Collection {type=FManagedArrayCollection}]
  properties {
   BackstopDistance=(bIsAnimatable=True,High=1,WeightMap="BackstopDistance")
   BackstopMeshName=()
   BackstopRadius=(bIsAnimatable=True,High=60,WeightMap="BackstopRadius")
   bWarnDuplicateProperty=True
  }
 }
 SimulationAnimDriveConfig {
  type=FChaosClothAssetSimulationAnimDriveConfigNode
  guid=#114
  inputPins [
   Collection {type=FManagedArrayCollection}
   AnimDriveStiffness.WeightMap {type=FString}
   AnimDriveDamping.WeightMap {type=FString}
  ]
  outputPins [Collection {type=FManagedArrayCollection}]
  properties {
   AnimDriveDamping=(bIsAnimatable=True,High=1,WeightMap="AnimDriveDamping")
   AnimDriveStiffness=(bIsAnimatable=True,High=1,WeightMap="AnimDriveStiffness")
   bWarnDuplicateProperty=True
  }
 }
 ImportedSimClothCollection {
  type=FGetDataflowVariableNode
  guid=#115
  outputPins [Value {type=FManagedArrayCollection}]
  properties {
   VariableName=ImportedSimClothCollection
   VariablePropertyBag=(Value=/Engine/Transient.PropertyBag_808350ac518c5f4c(ImportedSimClothCollection=()))
  }
 }
 ProxyDeformer {
  type=FChaosClothAssetProxyDeformerNode_v3
  guid=#116
  properties {InfluenceRadius=100, SelectionFilterSets=(()), bPreserveRenderTangents=True}
 }
 MergeClothCollections {type=FChaosClothAssetMergeClothCollectionsNode_v2, guid=#117}
 ReferenceBone {
  type=FChaosClothAssetReferenceBoneNode
  guid=#118
  inputPins [Collection {type=FManagedArrayCollection}]
  outputPins [Collection {type=FManagedArrayCollection}]
  properties {CalculateDefaultReferenceBone=(), ReferenceBone=(Name="pelvis")}
 }
 SkeletalMeshImport_Render {
  type=FChaosClothAssetSkeletalMeshImportNode_v2
  guid=#119
  properties {SectionIndex=1, SkeletalMesh=None, bImportRenderMesh=True, bImportSimMesh=False, bImportSingleSection=True}
 }
 RenderSkeletalMesh {
  type=FGetDataflowVariableNode
  guid=#120
  outputPins [Value {type="TObjectPtr<USkeletalMesh>"}]
  properties {VariableName=RenderSkeletalMesh, VariablePropertyBag=(Value=/Engine/Transient.PropertyBag_6efe5d0dc9f4d2bc(RenderSkeletalMesh=None))}
 }
 SimulationVelocityScaleConfig {
  type=FChaosClothAssetSimulationVelocityScaleConfigNode
  guid=#121
  inputPins [Collection {type=FManagedArrayCollection}]
  outputPins [Collection {type=FManagedArrayCollection}]
  properties {
   AngularVelocityScale=0.6
   FictitiousAngularScale=0
   FictitiousForcesModel=Legacy
   LinearVelocityScale=(X=0.6,Y=0.6,Z=0.5)
   MaxAngularAcceleration=12000
   MaxAngularVelocity=200
   MaxLinearAcceleration=(X=60000,Y=60000,Z=60000)
   MaxLinearVelocity=(X=1000,Y=1000,Z=1000)
   MaxVelocityScale=1
   VelocityScaleSpace=ReferenceBoneSpace
   bEnableAngularAccelerationClamping=False
   bEnableAngularVelocityClamping=False
   bEnableLinearAccelerationClamping=False
   bEnableLinearVelocityClamping=False
   bWarnDuplicateProperty=True
  }
 }
 SimulationGravityConfig {
  type=FChaosClothAssetSimulationGravityConfigNode
  guid=#122
  properties {GravityOverrideImported=(ImportedValue=(X=0,Y=0,Z=-980))}
 }
 ReRouteNode_0 {
  type=FDataflowReRouteNode
  guid=#123
  inputPins [Value {type=FManagedArrayCollection}]
  outputPins [Value {type=FManagedArrayCollection}]
 }
 SimulationPressureConfig {
  type=FChaosClothAssetSimulationPressureConfigNode
  guid=#124
  inputPins [
   Collection {type=FManagedArrayCollection}
   Pressure.WeightMap {type=FString}
  ]
  outputPins [Collection {type=FManagedArrayCollection}]
  properties {Pressure=(bIsAnimatable=True,High=1,WeightMap="Pressure"), bWarnDuplicateProperty=True}
 }
 ProceduralSelection {
  type=FChaosClothAssetProceduralSelectionNode
  guid=#125
  inputPins [
   Collection {type=FManagedArrayCollection}
   ConversionInputName.StringValue {type=FString}
  ]
  outputPins [
   Collection {type=FManagedArrayCollection}
   OutputName {type=FString}
  ]
  properties {ConversionInputName=(), Group=(Name="SimVertices3D"), OutputName=AllSimVertices3D_ProxyDeformer, SelectionType=SelectAll}
 }
 ProceduralSelection_0 {
  type=FChaosClothAssetProceduralSelectionNode
  guid=#126
  inputPins [
   Collection {type=FManagedArrayCollection}
   ConversionInputName.StringValue {type=FString}
  ]
  outputPins [
   Collection {type=FManagedArrayCollection}
   OutputName {type=FString}
  ]
  properties {ConversionInputName=(), Group=(Name="RenderVertices"), OutputName=AllRenderVertices_ProxyDeformer, SelectionType=SelectAll}
 }
 SkinningBlend {
  type=FChaosClothAssetSkinningBlendNode
  guid=#127
  inputPins [
   Collection {type=FManagedArrayCollection}
   KinematicVertices3D.StringValue {type=FString}
  ]
  outputPins [
   Collection {type=FManagedArrayCollection}
   SkinningBlendName {type=FString}
  ]
  properties {KinematicVertices3D=(), SkinningBlendName=RenderDeformerSkinningBlend, bUseSmoothTransition=True}
 }
 LegacyMask_MaxDistance {type=FChaosClothAssetWeightMapNode, guid=#128, properties {OutputName=(StringValue="MaxDistance")}}
 LegacyMask_BackstopDistance {type=FChaosClothAssetWeightMapNode, guid=#129, properties {OutputName=(StringValue="BackstopDistance")}}
 LegacyMask_BackstopRadius {type=FChaosClothAssetWeightMapNode, guid=#130, properties {OutputName=(StringValue="BackstopRadius")}}
 LegacyMask_AnimDriveStiffness {type=FChaosClothAssetWeightMapNode, guid=#131, properties {OutputName=(StringValue="AnimDriveStiffness")}}
 LegacyMask_Drag {type=FChaosClothAssetWeightMapNode, guid=#132, properties {OutputName=(StringValue="Drag")}}
 LegacyMask_Lift {type=FChaosClothAssetWeightMapNode, guid=#133, properties {OutputName=(StringValue="Lift")}}
]
